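# Physics asset for the ButterPlate axe mesh, a cosmetic melee pickaxe weapon in the BRCosmetics content.
# It defines no constraints, so it sets up no jointed or ragdoll-style motion between bodies.
# The file shows no body or collision shape setup.
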PhysicsAsset /BRCosmetics/Weapons/FORT_Melee/Pickaxe_ButterPlate/Meshes/ButterPlate_Axe_PhysicsAsset
constraints []
